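# Editor utility widget for the quest editor's objective editing tools, providing a "Common Tag Queries" button for quest tag query selection.
# It derives from the native FortCommonQuestTagQueryMenu class in FortniteEditor, with a QuestEditorButton as root bound to the inherited Button_TagQueryMenu variable.
# The event graph holds only unconnected default events, so any menu behavior comes from the C++ parent.
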WidgetBlueprint /Game/NeverCook/EditorUtilities/QuestTools/QuestEditor/QuestObjectiveEditors/FortCommonQuestTagQueryMenu
graphs [EventGraph {nodes [
   Event PreConstruct()
   Event Construct()
   Event Tick()
  ]}]
variables []
widgets {
 info {
  inheritedWidgetCount=1
  parentClass=/Script/FortniteEditor.FortCommonQuestTagQueryMenu
  rootWidgetClass=/Game/NeverCook/EditorUtilities/QuestTools/QuestEditor/QuestEditorButtons/QuestEditorButton.QuestEditorButton_C
  widgetCount=1
 }
 widgets [Button_TagQueryMenu {
   widgetClassPath=/Game/NeverCook/EditorUtilities/QuestTools/QuestEditor/QuestEditorButtons/QuestEditorButton
   bInherited=true
   bIsVariable=true
   widget=Button_TagQueryMenu
   properties {text=Common Tag Queries}
  }]
}
