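# Editor utility widget in the Fortnite quest editor tools for adding filters to a quest objective, built on the C++ FortQuestObjectiveFiltersEditor class.
# A header bar holds New and Paste filter buttons above a filter list box; the Paste button and list box are inherited bindings.
# The graph only forwards New clicks to AddQuestItemObjectiveFilter.
# Quest definition and objective index variables are declared but unused in this graph.
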EditorUtilityWidgetBlueprint /Game/NeverCook/EditorUtilities/QuestTools/QuestEditor/QuestObjectiveEditors/ObjectiveFilterEditors/FortObjectiveFiltersEditor
graphs [EventGraph {nodes [
   "On Clicked (Button_NewObjectiveFilter)"!() -> (OutputDelegate, then, Button:Common Button Base Object Reference) {is=title K2Node_ComponentBoundEvent}
   > AddQuestItemObjectiveFilter(self:Fort Quest Objective Filters Editor Object Reference)
  ]}]
variables [
 IndexInQuestObjectiveArray {type=integer}
 QuestObject {type=FortQuestItemDefinition, default=null}
]
widgets {
 info {
  inheritedWidgetCount=2
  parentClass=/Script/FortniteEditor.FortQuestObjectiveFiltersEditor
  rootWidgetClass=&VerticalBox
  widgetCount=7
 }
 widgets [
  VerticalBox_135 {widgetClassPath=&VerticalBox, widget=VerticalBox_135}
  HorizontalBox_78 {widgetClassPath=&HorizontalBox, parent=VerticalBox_135, widget=HorizontalBox_78, slot {class=VerticalBoxSlot}}
  Border_37 {
   widgetClassPath=&Border
   parent=HorizontalBox_78
   widget=Border_37
   properties {horizontalAlignment=HAlign_Left, brushColor=<rgba 0.0176 0.0176 0.0176 1>, padding=<ltrb 0 0 0 0>}
   slot {class=HorizontalBoxSlot, properties {size {sizeRule=Fill, value=1}}}
  }
  HorizontalBox_1 {
   widgetClassPath=&HorizontalBox
   parent=Border_37
   widget=HorizontalBox_1
   slot {class=BorderSlot, properties {horizontalAlignment=HAlign_Left, padding=<ltrb 0 0 0 0>}}
  }
  Button_NewObjectiveFilter {
   widgetClassPath=/Game/NeverCook/EditorUtilities/QuestTools/QuestEditor/QuestEditorButtons/QuestEditorButton
   bIsVariable=true
   parent=HorizontalBox_1
   widget=Button_NewObjectiveFilter
   properties {
    bShowIcon=true
    style=/Game/NeverCook/EditorUtilities/QuestTools/QuestEditor/QuestEditorButtons/Styles/QuestEditorButtonStyle_Primary.QuestEditorButtonStyle_Primary_C
    text=New Objective Filter
    toolTipText=Create a new filter for the objective
    iconBrush {
     bIsDynamicallyLoaded=false
     drawAs=Image
     imageType=NoImage
     mirroring=NoMirror
     resourceName=None
     resourceObject=/Game/UI/Foundation/Textures/Buttons/T-SlotIcon-Add-32
     tiling=NoTile
     imageSize=<18 18>
     margin=<ltrb 0 0 0 0>
     outlineSettings {
      bUseBrushTransparency=false
      roundingType=HalfHeightRadius
      width=0
      color {colorUseRule=UseColor_Specified, specifiedColor=<rgba 0 0 0 0>}
      cornerRadii=<0 0 0 0>
     }
     tintColor {colorUseRule=UseColor_Specified, specifiedColor=<rgba 1 1 1 1>}
     uVRegion {bIsValid=false, max=<0 0>, min=<0 0>}
    }
   }
   slot {class=HorizontalBoxSlot, properties {horizontalAlignment=HAlign_Left, verticalAlignment=VAlign_Center, padding=<ltrb 5 5 5 5>}}
  }
  Button_PasteObjectiveFilter {
   widgetClassPath=&QuestEditorButton
   bInherited=true
   bIsVariable=true
   parent=HorizontalBox_1
   widget=Button_PasteObjectiveFilter
   properties {
    bShowIcon=true
    style=&QuestEditorButtonStyle_Primary_C
    text=Paste Objective Filter
    toolTipText=Create a new filter for the objective
    iconBrush {
     bIsDynamicallyLoaded=false
     drawAs=Image
     imageType=NoImage
     mirroring=NoMirror
     resourceName=None
     resourceObject=&T-SlotIcon-Add-32
     tiling=NoTile
     imageSize=<18 18>
     margin=<ltrb 0 0 0 0>
     outlineSettings {
      bUseBrushTransparency=false
      roundingType=HalfHeightRadius
      width=0
      color {colorUseRule=UseColor_Specified, specifiedColor=<rgba 0 0 0 0>}
      cornerRadii=<0 0 0 0>
     }
     tintColor {colorUseRule=UseColor_Specified, specifiedColor=<rgba 1 1 1 1>}
     uVRegion {bIsValid=false, max=<0 0>, min=<0 0>}
    }
   }
   slot {class=HorizontalBoxSlot, properties {horizontalAlignment=HAlign_Left, verticalAlignment=VAlign_Center, padding=<ltrb 0 5 0 5>}}
  }
  VerticalBox_ObjectiveFilters {
   widgetClassPath=&VerticalBox
   bInherited=true
   parent=VerticalBox_135
   widget=VerticalBox_ObjectiveFilters
   slot {class=VerticalBoxSlot, properties {size {sizeRule=Fill, value=1}}}
  }
 ]
}
